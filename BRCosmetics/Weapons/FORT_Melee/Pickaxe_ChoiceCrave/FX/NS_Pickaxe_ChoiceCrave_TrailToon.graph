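# Niagara system that draws the layered toon swing trail for the ChoiceCrave pickaxe.
# Four ribbon emitters (dark, glow, red, tip) build anim trails from skeletal mesh socket pairs; the tip adds a second ribbon renderer and derived glow color and width.
# Spawn alpha is cut off by system age, scaled per User.SwingSectionIndex to time each swing section.
# Update curves fade color and alpha, with time-of-day scaling, glow width taper and animated material parameters on the dark layer.
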NiagaraSystem /BRCosmetics/Weapons/FORT_Melee/Pickaxe_ChoiceCrave/FX/NS_Pickaxe_ChoiceCrave_TrailToon
dependencies {
 usedDataInterfaces [
  &NiagaraDataInterfaceCurve
  &NiagaraDataInterfaceColorCurve
  &NiagaraDataInterfaceSkeletalMesh
  &NiagaraDataInterfaceSocketReader
 ]
 usedDynamicInputs [&FN_Select_Float_from_Time_Of_Day]
 usedModules [
  &self:SystemSpawnScript.SystemScriptSource.SystemScriptGraph.NiagaraNodeAssignment_0.SetVariables_#1
  &SystemState
  &self:SystemSpawnScript.SystemScriptSource.SystemScriptGraph.NiagaraNodeAssignment_1.SetVariables_#2
 ]
}
emitters {
 TrailBlack {
  emitterUpdateScript [
   &EmitterState()
   &FN_SpawnAnimTrail(Switch on ENiagara_AnimTrail_Source:ENiagara_AnimTrail_Source=::NewEnumerator1)
  ]
  particleSpawnScript [
   &FN_InitializeAnimTrailBoneIndecies()
   &FN_InitializeAnimTrail(
    Color=&MultiplyLinearColor(
     Linear Color=<rgba 0.024 0.0267 0.04 0.5>
     Scale Factor=&MakeVector4FromVectorAndFloat(
      "Vector (RGB)"=<1 1 1>
      "Float (Alpha)"=&FloatFromCurve(
       FloatCurve=DataInterface() {propertyValues {Curve {keys [<at 0.09 1>,<at 0.12 0>]}}}
       CurveIndex=&Multiply_Float(
        A=$System.Age
        B=&MergeHelper_FloatSwitch(0=1.1, 1=1.17, 2=1, 3=1, Index=$User.SwingSectionIndex, Num=::NewEnumerator2)
       )
      )
     )
    )
    Lifetime=0.6
    SkeletalMesh=$System.Skeletal Mesh
   )
  ]
  particleUpdateScript [
   &ParticleState()
   &FN_UpdateAnimTrail(SkeletalMesh=$System.Skeletal Mesh)
   &ScaleColor(
    Scale Alpha=&FloatFromCurve(
     FloatCurve=DataInterface() {propertyValues {Curve {keys [<at 0.3 1>,<at 1 0>]}}}
     Scale Curve=&FloatFromCurve(
      FloatCurve=DataInterface() {propertyValues {Curve {keys [<at 0.12 1>,<at 0.2 0>]}}}
      CurveIndex=&Multiply_Float(
       A=$System.Age
       B=&MergeHelper_FloatSwitch(0=1.1, 1=1.17, 2=1, 3=1, Index=$User.SwingSectionIndex, Num=::NewEnumerator2)
      )
      Scale Curve=$System.ToDScalar
     )
    )
    Scale RGB=&VectorFromFloat(Value=$System.ToDScalar)
    ScaleA=true
    ScaleRGB=true
   )
   &DynamicMaterialParameters(
    Index 0 Param 1=0.1
    Index 0 Param 2=&FloatFromCurve(FloatCurve=DataInterface() {propertyValues {Curve {defaultValue=1, keys [{tangentMode=RCTM_User, time=0, value=-0.8}, {tangentMode=RCTM_User, time=0.6, value=0}]}}})
    Index 0 Param 3=&FloatFromCurve(
     FloatCurve=DataInterface() {propertyValues {Curve {defaultValue=1, keys [{tangentMode=RCTM_User, time=0, value=0}, {tangentMode=RCTM_User, time=1, value=1}]}}}
     Scale Curve=1.25
    )
    Module.Index0Float0Write=true
    Module.Index0Float1Write=true
    Module.Index0Float2Write=true
    Module.Index0Float3Write=true
   )
  ]
  rendererClasses [&NiagaraRibbonRendererProperties]
  renderers [{rendererClass=&NiagaraRibbonRendererProperties}]
 }
 TrailGlow {
  emitterUpdateScript [
   &EmitterState()
   &FN_SpawnAnimTrail(AnimTrailWidthMode=::FromFirst, Switch on ENiagara_AnimTrail_Source:ENiagara_AnimTrail_Source=::NewEnumerator1)
  ]
  particleSpawnScript [
   &FN_InitializeAnimTrailBoneIndecies(First Bone or Socket Index=1, Second Bone or Socket Index=0)
   &FN_InitializeAnimTrail(
    AnimTrailWidth=0.5
    Color=&MultiplyLinearColor(
     Linear Color=<rgba 1 0.1 0.13 0.5>
     Scale Factor=&MakeVector4FromVectorAndFloat(
      "Vector (RGB)"=<1 1 1>
      "Float (Alpha)"=&FloatFromCurve(
       FloatCurve=DataInterface() {propertyValues {Curve {keys [<at 0.1 1>,<at 0.12 0>]}}}
       CurveIndex=&Multiply_Float(
        A=$System.Age
        B=&MergeHelper_FloatSwitch(0=1.1, 1=1.17, 2=1, 3=1, Index=$User.SwingSectionIndex, Num=::NewEnumerator2)
       )
      )
     )
    )
    Lifetime=0.55
    SkeletalMesh=$System.Skeletal Mesh
   )
  ]
  particleUpdateScript [
   &ParticleState()
   &FN_UpdateAnimTrail(SkeletalMesh=$System.Skeletal Mesh)
   &ScaleColor(
    Scale Alpha=&FloatFromCurve(FloatCurve=DataInterface() {propertyValues {Curve {keys [<at 0 1>,<at 1 0>]}}}, Scale Curve=$System.ToDScalar)
    Scale RGB=&VectorFromFloat(Value=&FloatFromCurve(FloatCurve=DataInterface() {propertyValues {Curve {keys [<at 0 3>,<at 0.1 0.2>]}}}, CurveIndex=$System.Age))
    ScaleA=true
    ScaleRGB=true
   ) {module=ScaleColor001}
   &ScaleRibbonWidth(Ribbon Width Scale=&FloatFromCurve(
     FloatCurve=DataInterface() {propertyValues {Curve {keys [
         {arriveTangent=-2.7527, interpMode=RCIM_Cubic, leaveTangent=-2.7527, tangentMode=RCTM_User, time=0, value=1}
         {interpMode=RCIM_Cubic, tangentMode=RCTM_User, time=1, value=0}
        ]}}}
     Scale Curve=1.8
    ))
  ]
  rendererClasses [&NiagaraRibbonRendererProperties]
  renderers [{rendererClass=&NiagaraRibbonRendererProperties}]
 }
 TrailRed {
  emitterUpdateScript [
   &EmitterState()
   &FN_SpawnAnimTrail(Switch on ENiagara_AnimTrail_Source:ENiagara_AnimTrail_Source=::NewEnumerator1)
  ]
  particleSpawnScript [
   &FN_InitializeAnimTrailBoneIndecies(First Bone or Socket Index=2)
   &FN_InitializeAnimTrail(
    Color=&MultiplyLinearColor(
     Linear Color=<rgba 2 0.02 0.152 0.8>
     Scale Factor=&MakeVector4FromVectorAndFloat(
      "Vector (RGB)"=<1 1 1>
      "Float (Alpha)"=&FloatFromCurve(
       FloatCurve=DataInterface() {propertyValues {Curve {keys [<at 0.08 1>,<at 0.12 0>]}}}
       CurveIndex=&Multiply_Float(
        A=$System.Age
        B=&MergeHelper_FloatSwitch(0=1.1, 1=1.17, 2=1, 3=1, Index=$User.SwingSectionIndex, Num=::NewEnumerator2)
       )
      )
     )
    )
    Lifetime=0.35
    SkeletalMesh=$System.Skeletal Mesh
   )
  ]
  particleUpdateScript [
   &ParticleState()
   &FN_UpdateAnimTrail(SkeletalMesh=$System.Skeletal Mesh)
   &ScaleColor(
    Scale Alpha=&FloatFromCurve(FloatCurve=DataInterface() {propertyValues {Curve {keys [<at 0 1>,<at 1 0>]}}}, Scale Curve=$System.ToDScalar)
    Scale RGB=&VectorFromFloat(Value=&FloatFromCurve(FloatCurve=DataInterface() {propertyValues {Curve {keys [<at 0 3>,<at 0.8 0.4>]}}}))
    ScaleA=true
    ScaleRGB=true
   )
  ]
  rendererClasses [&NiagaraRibbonRendererProperties]
  renderers [{rendererClass=&NiagaraRibbonRendererProperties}]
 }
 TrailTip {
  emitterUpdateScript [
   &EmitterState()
   &FN_SpawnAnimTrail(Switch on ENiagara_AnimTrail_Source:ENiagara_AnimTrail_Source=::NewEnumerator1)
  ]
  particleSpawnScript [
   &FN_InitializeAnimTrailBoneIndecies(First Bone or Socket Index=3)
   &FN_InitializeAnimTrail(
    Color=&MultiplyLinearColor(
     Linear Color=<rgba 1 0.2502 0.624 0.8>
     Scale Factor=&MakeVector4FromVectorAndFloat(
      "Vector (RGB)"=<1 1 1>
      "Float (Alpha)"=&FloatFromCurve(
       FloatCurve=DataInterface() {propertyValues {Curve {keys [{arriveTangent=-65.0147, interpMode=RCIM_Cubic, leaveTangent=-65.0147, tangentMode=RCTM_User, time=0.08, value=1}, <at 0.13 0>]}}}
       CurveIndex=&Multiply_Float(
        A=$System.Age
        B=&MergeHelper_FloatSwitch(0=1.1, 1=1.17, 2=1, 3=1, Index=$User.SwingSectionIndex, Num=::NewEnumerator2)
       )
      )
     )
    )
    Lifetime=0.42
    SkeletalMesh=$System.Skeletal Mesh
   )
  ]
  particleUpdateScript [
   &ParticleState()
   &FN_UpdateAnimTrail(SkeletalMesh=$System.Skeletal Mesh)
   &ScaleColor(
    Scale Alpha=&FloatFromCurve(FloatCurve=DataInterface() {propertyValues {Curve {keys [<at 0 1>,<at 1 0>]}}}, Scale Curve=$System.ToDScalar)
    Scale RGB=&MakeVectorFromLinearColorRGB(Color=&ColorFromCurve(ColorCurve=DataInterface() {propertyValues {
        AlphaCurve {keys [<at 0 1 RCIM_Cubic>,<at 1 0 RCIM_Cubic>]}
        BlueCurve {keys [<at 0 8.3333 RCIM_Cubic>,<at 0.5033 0.0646 RCIM_Cubic>]}
        GreenCurve {keys [<at 0 0 RCIM_Cubic>,<at 0.5033 0.025 RCIM_Cubic>]}
        RedCurve {keys [<at 0 50 RCIM_Cubic>,<at 0.5033 0.5 RCIM_Cubic>]}
       }}))
    ScaleA=true
    ScaleRGB=true
   )
   /SetVariables(
    Particles.GlowColor=&MultiplyLinearColor(
     Linear Color=$Particles.Color
     Scale Factor=&MakeVector4FromVectorAndFloat("Vector (RGB)"=<1 1 1>, "Float (Alpha)"=0.05)
    )
    Particles.GlowWidth:NiagaraFloat=&Multiply_Float(A=$Particles.RibbonWidth, B=1.5)
   ) {isSetParameters=true}
  ]
  rendererClasses [&NiagaraRibbonRendererProperties]
  renderers [
   {rendererClass=&NiagaraRibbonRendererProperties}
   {rendererClass=&NiagaraRibbonRendererProperties, rendererIndex=1}
  ]
 }
}
summary {
 systemName=NS_Pickaxe_ChoiceCrave_TrailToon
 emitters [
  {emitterName=TrailRed, rendererClasses [&NiagaraRibbonRendererProperties]}
  {emitterName=TrailTip, rendererClasses [&NiagaraRibbonRendererProperties]}
  {emitterName=TrailBlack, rendererClasses [&NiagaraRibbonRendererProperties]}
  {emitterName=TrailGlow, rendererClasses [&NiagaraRibbonRendererProperties]}
 ]
 userVariables [
  User.bIsFrontEnd {type=NiagaraBool}
  User.bIsFrontEndPreview {type=NiagaraBool}
  User.FrontEndScale {type=NiagaraFloat, defaultValue=0.6, description=""}
  User.RibbonIDOffset {type=NiagaraInt32}
  User.SwingSectionIndex {type=NiagaraInt32}
 ]
}
